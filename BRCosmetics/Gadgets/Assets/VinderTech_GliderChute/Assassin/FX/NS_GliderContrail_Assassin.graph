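# Niagara ribbon contrail for the Assassin glider chute, drawing short-lived red translucent trails from the glider's skeletal mesh.
# Its one emitter, Contrail_Left, splits particles into two spawn groups, each indexing a bone or socket on User.SkeletalMesh.
# A steady spawn rate feeds the ribbons while spawn-per-unit is disabled, and a curve fades alpha in and out.
# User.bIsFrontEnd adds a constant velocity so trails stream in the front end without inheriting mesh motion.
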NiagaraSystem /BRCosmetics/Gadgets/Assets/VinderTech_GliderChute/Assassin/FX/NS_GliderContrail_Assassin
dependencies {usedDataInterfaces [&NiagaraDataInterfaceCurve], usedDynamicInputs [], usedModules [&SystemState]}
emitters {Contrail_Left {
  emitterUpdateScript [
   &EmitterState(
    Delay First Loop Only=true
    Inactive Response=::NewEnumerator1
    Life Cycle Mode=::NewEnumerator1
    Loop Delay=0.25
    Loop Duration=0.1
    UseLoopDelay=true
   )
   &SpawnRate(SpawnRate=150, Use Spawn Probability=false)
   &SpawnPerUnit(Module.Use Max Movement Threshold=true, Module.Use Movement Tolerance=false, Spawn Per Unit=10, Use Spawn Probability=false) {enabled=false}
  ]
  particleSpawnScript [
   &FN_SetSpawnGroup(Number of Spawn Groups=2)
   &InitializeRibbon(
    Lifetime=0.4
    Module.Write Color=true
    Module.Write Lifetime=true
    Module.Write Mass=true
    Module.Write Material Random=true
    Module.Write Position=false
    Module.Write RibbonTwist=false
    Module.Write RibbonWidth=true
    Ribbon Width=20
   )
   &AddVelocity(Velocity=&MakeCustomVectorFromBool(Bool=$User.bIsFrontEnd, True Vector=<500 0 1>, False Vector=<0 0 0>))
   &Color(Color=&MakeLinearColorFromVectorAndFloat("Vector (RGB)"=<1 0 0>, "Float (Alpha)"=&ClampFloat(Float=0.5)))
   &SkeletalMeshLocation(
    Bone / Socket Index=$Particles.SpawnGroup
    Bone Sampling Mode=::NewEnumerator4
    Inherit Velocity=false
    MaskSpawnGroup=false
    Sampled Position Offset=<-2 0 -5>
    Skeletal Mesh=$User.SkeletalMesh
   )
  ]
  particleUpdateScript [
   &ParticleState()
   &SolveForcesAndVelocity(Clamp Velocity=false, Limit Acceleration=false)
   &ScaleColor(
    Scale Alpha=&FloatFromCurve(FloatCurve=DataInterface() {propertyValues {Curve {keys [
         {arriveTangent=8.4533, interpMode=RCIM_Cubic, leaveTangent=8.4533, tangentMode=RCTM_User, time=0, value=0}
         <at 0.3835 0.9839 RCIM_Cubic>
         <at 1 0 RCIM_Cubic>
        ]}}})
    ScaleA=true
    ScaleRGB=true
   )
  ]
  rendererClasses [&NiagaraRibbonRendererProperties]
  renderers [{rendererClass=&NiagaraRibbonRendererProperties}]
 }}
summary {
 systemName=NS_GliderContrail_Assassin
 emitters [{emitterName=Contrail_Left, rendererClasses [&NiagaraRibbonRendererProperties]}]
 userVariables [
  User.bIsFrontEnd {type=NiagaraBool}
  User.SkeletalMesh {
   type=NiagaraDataInterfaceSkeletalMesh
   defaultValue {
    struct=NiagaraExt_VariableValue_DataInterface
    value {dataInterface=&self:NiagaraDataInterfaceSkeletalMesh_0}
   }
  }
 ]
}
