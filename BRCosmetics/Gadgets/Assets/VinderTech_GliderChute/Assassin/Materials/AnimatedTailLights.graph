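# Material function for the Assassin glider chute that overrides emissive color to animate tail lights for idle, turning and braking.
# The emissive texture is scaled to default intensity, blended by isTurning to a sequenced signal, then by isBraking to full brightness.
# The signal sphere-masks the FX texture's green gradient around TailLightSequence, with TailLightDirection picking the side by U coordinate.
# The time-based blink and red-channel blend are left unconnected.
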MaterialFunction /BRCosmetics/Gadgets/Assets/VinderTech_GliderChute/Assassin/Materials/AnimatedTailLights
graph {nodes [
  Frac(None=Multiply(A=Time(), B=ScalarParameter() {defaultValue=0.7, group=TailLights, parameterName=TailLightBlinkRate}))
  n1 = FunctionInput() {bUsePreviewValueAsDefault=true, inputName=Material, inputType=FunctionInput_MaterialAttributes}
  n2 = TextureSample(UVs=TextureCoordinate()) {texture=/BRCosmetics/Gadgets/Assets/VinderTech_GliderChute/Assassin/Textures/Assassin_GliderChute_Parts_E}
  n3 = ScalarParameter() {defaultValue=0.7, group=TailLights, parameterName=DefaultTailLightIntensity}
  n4 = TextureSample(UVs=TextureCoordinate()) {samplerType=SAMPLERTYPE_Masks, texture=/BRCosmetics/Gadgets/Assets/VinderTech_GliderChute/Assassin/Textures/Assassin_GliderChute_Parts_FX}
  n5 = MaterialFunctionCall(Float3=FunctionInput(Preview=@n4.RGB) {bUsePreviewValueAsDefault=true, inputName=TaillLightTexture}) {materialFunction=&BreakOutFloat3Components}
  n6 = Ceil(None=@n5.G)
  n7 = SphereMask(
   A=Multiply(A=@n6, B=@n5.G)
   B=ScalarParameter() {parameterName=TailLightSequence}
   Radius=ScalarParameter() {defaultValue=0.25, group=TailLights, parameterName=Range}
   Hardness=Constant() {r=1}
  )
  n8 = ScalarParameter() {defaultValue=2, parameterName=TailLightBrightness}
  n9 = Multiply(A=Multiply(A=@n2.RGB, B=Multiply(A=@n6, B=@n7)), B=@n8)
  n10 = MaterialFunctionCall(Float2=TextureCoordinate()) {materialFunction=&BreakOutFloat2Components}
  n11 = Clamp(None=Ceil(None=Subtract(A=@n10.R)))
  n12 = LinearInterpolate(A=@n11, B=OneMinus(None=@n11), Alpha=ScalarParameter() {parameterName=TailLightDirection})
  n13 = LinearInterpolate(
   A=LinearInterpolate(A=Multiply(A=@n2.RGB, B=@n3), B=Multiply(A=@n9, B=@n12), Alpha=ScalarParameter() {group=TailLights, parameterName=isTurning})
   B=Multiply(A=@n2.RGB, B=@n8)
   Alpha=ScalarParameter() {defaultValue=1, group=TailLights, parameterName=isBraking}
  )
  FunctionOutput(None=SetMaterialAttributes!(
    MaterialAttributes=GetMaterialAttributes(None=@n1) {attributeGetTypes [#1]}.MaterialAttributes
    EmissiveColor=@n13
   ) {attributeSetTypes [#1]})
  LinearInterpolate(A=@n2.RGB, B=@n13, Alpha=@n5.R)
 ]}
