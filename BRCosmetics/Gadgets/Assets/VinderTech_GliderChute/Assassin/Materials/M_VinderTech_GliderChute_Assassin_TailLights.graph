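# Default Lit material for the tail lights of the Assassin VinderTech glider chute, enabled for skeletal meshes and mesh deformers.
# It uses material attributes, passed through a chain of shared functions with no other logic shown.
# The gun master shader function feeds Hologram, then the weapon skin FX function.
# AnimatedTailLights then adds the light effect, and the accessories tech art overrides function finishes the chain.
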Material /BRCosmetics/Gadgets/Assets/VinderTech_GliderChute/Assassin/Materials/M_VinderTech_GliderChute_Assassin_TailLights
graph {
 properties {
  bUseMaterialAttributes=true
  bUsedWithMeshDeformer=true
  bUsedWithSkeletalMesh=true
  pixelDepthOffsetMode=PDOM_Legacy
  usedShadingModels=Default Lit
 }
 nodes [
  n1 = MaterialFunctionCall!(In=MaterialFunctionCall!(UV 1 coordinates, MuzzleFalloff) {materialFunction=/Game/Packages/Fortnite_UniqueMaterials_1/Weapons/M_GunMasterShaderFunction}.Material Result) {materialFunction=/Game/Packages/Fortress_SharedMaterials/Base_Material_Functions/Hologram}
  MaterialFunctionCall(In=MaterialFunctionCall!(
    TaillLightTexture
    Material=MaterialFunctionCall!(Pre-Skinned World Normal, Gmap, Customization Mask, IsCharacter, IsStatsB, Material Attributes=@n1.Material Att) {materialFunction=/Game/Weapons/WeaponSkins/MaterialFunction/Weapon_SkinFXFunction}.Result
   ) {materialFunction=&/AnimatedTailLights}.Result) {materialFunction=&MF_TechArtOverrides_accessories}
 ]
}
